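# Niagara effect for the NitroFlow glider: bone-attached wing meshes, GPU droplets and soft caustic sprites.
# Left and right wing emitters chase a lagged target, clamp it by angle and arm length, then pass it to materials through dynamic parameters.
# A system Timeline blends wing pose and droplet velocity, and User.Deployed offsets the unfold animation.
# Variant and time-of-day switches adjust color or alpha; quality level adds wing pieces.
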NiagaraSystem /BRCosmetics/Gadgets/Assets/VinderTech_GliderChute/Glider_NitroFlow/FX/NS_Glider_Nitroflow
dependencies {
 usedDataInterfaces [
  &NiagaraDataInterfaceColorCurve
  &NiagaraDataInterfaceCurve
  &NiagaraDataInterfacePlatformSet
  &NiagaraDataInterfaceSpriteRendererInfo
 ]
 usedDynamicInputs [&SetBoolByFloatComparison, &Abs_Float]
 usedModules [
  &SystemState
  &Timeline
  &self:SystemSpawnScript.SystemScriptSource.SystemScriptGraph.NiagaraNodeAssignment_0.SetVariables_#1
  &self:SystemSpawnScript.SystemScriptSource.SystemScriptGraph.NiagaraNodeAssignment_1.SetVariables_#2
 ]
}
emitters {
 Caustic {
  emitterUpdateScript [
   &EmitterState(Life Cycle Mode=::NewEnumerator1, Loop Duration=1, UseLoopDelay=false)
   &SpawnRate(SpawnRate=2, Use Spawn Probability=false)
  ]
  particleSpawnScript [
   &FN_SetSpawnGroup(Number of Spawn Groups=2)
   /Niagara/Modules/Spawn/Initialization/InitializeParticle(
    Color=&MakeCustomLinearColorFromBool(Bool=$System.DefaultVarient, True Color=<rgba 0 1.4474 2 1>, False Color=<rgba 1 0.831 0.3297 1>)
    Lifetime=3.6
    Module.Write Color=true
    Module.Write Lifetime=true
    Module.Write Mass=true
    Module.Write Material Random=true
    Module.Write Position=true
    Module.Write Scale=false
    Module.Write SpriteRotation=true
    Module.Write SpriteSize=true
    Sprite Rotation=&RandomRangeFloat(Minimum=0, Maximum=360)
    Sprite Size=<50 50>
   )
   &DynamicMaterialParameters(
    Index 0 Param 1=1
    Index 0 Param 2=0.35
    Index 0 Param 3=0.35
    Index 0 Param 4=1
    Module.Param0WriteEnabled=true
    Module.Param1WriteEnabled=true
    Module.Param2WriteEnabled=true
    Module.Param3WriteEnabled=true
   )
   &CameraOffset(Camera Offset Amount=20)
  ]
  particleUpdateScript [
   &ParticleState()
   &SkeletalMeshLocation(
    Bone / Socket Index=&Add_Integer(A=0, B=&Multiply_Int(A=2, B=&SpawnGroup()))
    Bone Sampling Mode=::NewEnumerator4
    Inherit Velocity=false
    MaskSpawnGroup=false
    Sampled Position Offset=&Lerp_Vector(A=<0 0 0>, B=<0 0 0>, Alpha=&MakeFloatFromInt(Int=$Particles.SpawnGroup))
    Sampled Position Offset Coordinate Space:ENiagaraMeshTransforms=::NewEnumerator3
    Skeletal Mesh=$User.SkeletalMesh
   )
   /SetVariables(Particles.APosition:NiagaraPosition=$Particles.Position) {isSetParameters=true}
   &SkeletalMeshLocation(
    Bone / Socket Index=&Add_Integer(A=1, B=&Multiply_Int(A=2, B=&SpawnGroup()))
    Bone Sampling Mode=::NewEnumerator4
    Inherit Velocity=false
    MaskSpawnGroup=false
    Sampled Position Offset=&Lerp_Vector(A=<0 15 0>, B=<0 -15 0>, Alpha=&MakeFloatFromInt(Int=$Particles.SpawnGroup))
    Sampled Position Offset Coordinate Space:ENiagaraMeshTransforms=::NewEnumerator3
    Skeletal Mesh=$User.SkeletalMesh
   ) {module=SkeletalMeshLocation001}
   /SetVariables(Particles.BPosition:NiagaraPosition=$Particles.Position) {isSetParameters=true}
   /SetVariables(Particles.Position=&Lerp_Vector(A=$Particles.BPosition, B=$Particles.APosition, Alpha=&Multiply_Float(A=$Particles.NormalizedAge, B=1))) {isSetParameters=true}
   &SpriteRotationRate(Rotation Rate=10)
   &ScaleColor(
    Scale Alpha=&FloatFromCurve(FloatCurve=DataInterface() {propertyValues {Curve {keys [<at 0 0 RCIM_Cubic>,<at 0.5 1 RCIM_Cubic>,<at 1 0 RCIM_Cubic>]}}})
    Scale RGB=&MakeVectorFromLinearColorRGB(Color=&MakeCustomLinearColorFromBool(
      Bool=$System.DefaultVarient
      True Color=&ColorFromCurve(ColorCurve=DataInterface() {propertyValues {
         AlphaCurve {keys []}
         BlueCurve {keys [<at 0.25 1 RCIM_Cubic>,<at 0.75 1 RCIM_Cubic>]}
         GreenCurve {keys [<at 0.25 0.529 RCIM_Cubic>,<at 0.75 0.4107 RCIM_Cubic>]}
         RedCurve {keys [<at 0.25 0.6467 RCIM_Cubic>,<at 0.75 0.1805 RCIM_Cubic>]}
        }})
      False Color=&FNTimeOfDayColor(
       Night Color=&ColorFromCurve(ColorCurve=DataInterface() {propertyValues {
          AlphaCurve {keys []}
          BlueCurve {keys [<at 0.25 1 RCIM_Cubic>,<at 0.75 1 RCIM_Cubic>]}
          GreenCurve {keys [<at 0.25 0.6013 RCIM_Cubic>,<at 0.75 0.4568 RCIM_Cubic>]}
          RedCurve {keys [<at 0.25 0.5209 RCIM_Cubic>,<at 0.75 0.4826 RCIM_Cubic>]}
         }})
       Morning Color=&ColorFromCurve(ColorCurve=DataInterface() {propertyValues {
          AlphaCurve {keys []}
          BlueCurve {keys [<at 0.25 0.6166 RCIM_Cubic>,<at 0.75 0.6261 RCIM_Cubic>]}
          GreenCurve {keys [<at 0.25 0.921 RCIM_Cubic>,<at 0.75 0.5783 RCIM_Cubic>]}
          RedCurve {keys [<at 0.25 1 RCIM_Cubic>,<at 0.75 1 RCIM_Cubic>]}
         }})
       Day Color=&ColorFromCurve(ColorCurve=DataInterface() {propertyValues {
          AlphaCurve {keys []}
          BlueCurve {keys [<at 0.25 0.6166 RCIM_Cubic>,<at 0.75 0.6261 RCIM_Cubic>]}
          GreenCurve {keys [<at 0.25 0.921 RCIM_Cubic>,<at 0.75 0.5783 RCIM_Cubic>]}
          RedCurve {keys [<at 0.25 1 RCIM_Cubic>,<at 0.75 1 RCIM_Cubic>]}
         }})
       Evening Color=&ColorFromCurve(ColorCurve=DataInterface() {propertyValues {
          AlphaCurve {keys []}
          BlueCurve {keys [<at 0.25 0.6166 RCIM_Cubic>,<at 0.75 0.6261 RCIM_Cubic>]}
          GreenCurve {keys [<at 0.25 0.921 RCIM_Cubic>,<at 0.75 0.5783 RCIM_Cubic>]}
          RedCurve {keys [<at 0.25 1 RCIM_Cubic>,<at 0.75 1 RCIM_Cubic>]}
         }})
      )
     ))
    ScaleA=true
    ScaleRGB=true
   )
  ]
  rendererClasses [&NiagaraSpriteRendererProperties]
  renderers [{rendererClass=&NiagaraSpriteRendererProperties}]
 }
 Drops {
  simTarget=GPUComputeSim
  emitterUpdateScript [
   &EmitterState(Inactive Response=::NewEnumerator1, Life Cycle Mode=::NewEnumerator1, Loop Behavior=::NewEnumerator1, Loop Duration=3600, UseLoopDelay=false)
   &SpawnRate(
    SpawnGroup:NiagaraInt32=&RandomRangeInt(Minimum=0, Maximum=1)
    SpawnRate=32
    Use Spawn Probability=false
   )
  ]
  particleSpawnScript [
   &FN_SetSpawnGroup(Number of Spawn Groups=2)
   /Niagara/Modules/Spawn/Initialization/V2/InitializeParticle(
    Color=&MakeCustomLinearColorFromBool(
     Bool=$System.DefaultVarient
     True Color=&MakeCustomLinearColorFromBool(Bool=&RandomBool(), True Color=<rgba 0 1.5735 2 0.8>, False Color=<rgba 0.8523 0.5146 2 0.8>)
     False Color=&MakeCustomLinearColorFromBool(Bool=&RandomBool(), True Color=<rgba 0.1231 0.3666 2 0.8>, False Color=<rgba 2 1.5085 0.8489 0.8>)
    )
    Lifetime Max=0.8
    Lifetime Min=0.8
    Lifetime Mode=::NewEnumerator1
    Sprite Rotation Mode=::NewEnumerator2
    Sprite Size Mode=::NewEnumerator1
    Uniform Sprite Size Min=10
    UsePositionOffset=false
    Write Material Random=true
   ) {module=InitializeParticle001}
   &SkeletalMeshLocation(
    Apply To Particle Position=false
    Bone Sampling Mode=::NewEnumerator0
    Inherit Velocity=true
    "Inherit Velocity (Scale)"=0.1
    Inherited Velocity Speed Limit=1
    Interpolate Mesh Sampling=false
    MaskSpawnGroup=true
    Module.Override Randomness=false
    Sampled Position Offset Coordinate Space:ENiagaraMeshTransforms=::NewEnumerator2
    Sampled Velocity Coordinate Space:ENiagaraCoordinateSpace=::Local
    Skeletal Mesh=$User.SkeletalMeshRight
    Use Sampled Mesh Attribute as Ribbon ID=false
    VelocitySpeedLimit=true
   )
   &SkeletalMeshLocation(
    Apply To Particle Position=false
    Bone Sampling Mode=::NewEnumerator0
    Inherit Velocity=true
    "Inherit Velocity (Scale)"=0.1
    Inherited Velocity Speed Limit=1
    Interpolate Mesh Sampling=false
    MaskSpawnGroup=true
    Module.Override Randomness=false
    Sampled Position Offset Coordinate Space:ENiagaraMeshTransforms=::NewEnumerator2
    Sampled Velocity Coordinate Space:ENiagaraCoordinateSpace=::Local
    Skeletal Mesh=$User.SkeletalMeshLeft
    Spawn Group Mask=1
    Use Sampled Mesh Attribute as Ribbon ID=false
    VelocitySpeedLimit=true
   ) {module=SkeletalMeshLocation001}
   /SetVariables(Particles.Velocity=&Add_Vector(
     A=$Output.SkeletalMeshLocation.SampledVelocity
     B=&Lerp_Vector(
      A=&RandomRangeVector(Minimum=<-10 -40 -10>, Maximum=<10 -50 10>)
      B=&RandomRangeVector(Minimum=<0 -150 0>, Maximum=<1 -150 1>)
      Alpha=$System.Timeline.NormalizedTime
     )
    )) {isSetParameters=true}
   /SetVariables(Particles.SpriteSize=&Multiply_Vector2(
     A=<2 2>
     B=&Vector2DFromFloat(Value=&UniformRangedFloat(Minimum=1, Maximum=2))
    )) {isSetParameters=true}
   &SubUVAnimation(
    End Frame Range Override=3
    Evaluation Type=::NewEnumerator0
    "Sprite Renderer "=DataInterface() {propertyValues {SpriteRenderer=&self:Drops_1.NiagaraSpriteRendererProperties_1}}
    SubUV Animation Mode:ENiagara_SubUVLookupModeV2=::NewEnumerator2
    UseEndFrame=true
    UseStartFrame=false
   )
   /SetVariables(Particles.Position=&MaskVectorBySpawnGroup(Masked Vector=$Output.SkeletalMeshLocation.SampledPosition, Passthrough Vector=$Output.SkeletalMeshLocation001.SampledPosition)) {isSetParameters=true}
   &CameraOffset(Camera Offset Amount=&RandomRangeFloat(Minimum=-40, Maximum=-20))
  ]
  particleUpdateScript [
   &ParticleState()
   &GravityForce(Gravity=<0 0 -20>)
   &Drag(Drag=0, Use Linear Drag=true, Use Rotational Drag=false)
   &SolveForcesAndVelocity(Clamp Velocity=false, Limit Acceleration=false)
   &ScaleColor(
    Scale Alpha=&FloatFromCurve(FloatCurve=DataInterface() {propertyValues {Curve {keys [
         <at 0 0 RCIM_Cubic>
         {arriveTangent=1.6041, leaveTangent=1.6041, time=0.05, value=1.0113}
         {interpMode=RCIM_Cubic, leaveTangent=-1.0633, time=0.75, value=1}
         <at 0.99 0 RCIM_Cubic>
        ]}}})
    ScaleA=true
    ScaleRGB=true
   )
   &ScaleSpriteSize(Scale Factor=&Vector2DFromFloat(Value=&FloatFromCurve(
      FloatCurve=DataInterface() {propertyValues {Curve {keys [
          {arriveTangent=-0.4919, interpMode=RCIM_Cubic, leaveTangent=-0.4919, tangentMode=RCTM_User, time=0.25, value=2}
          {arriveTangent=-3.3301, interpMode=RCIM_Cubic, leaveTangent=-3.3302, tangentMode=RCTM_User, time=1, value=0}
         ]}}}
      Scale Curve=0.65
     )))
  ]
  rendererClasses [&NiagaraSpriteRendererProperties]
  renderers [{rendererClass=&NiagaraSpriteRendererProperties}]
 }
 WingLeft {
  emitterUpdateScript [
   &EmitterState(
    Inactive Response=::NewEnumerator1
    Life Cycle Mode=::NewEnumerator1
    Loop Behavior=::NewEnumerator1
    Loop Duration Mode=::NewEnumerator1
    UseLoopDelay=true
   )
   &SpawnBurst_Instantaneous(Spawn Count=9, Use Spawn Probability=false, UseLoopCountLimit=false)
   &SpawnBurst_Instantaneous(
    Spawn Count=&MakeCustomIntFromBool(
     Bool=&IsPlatformSetActive(New Platform Set=DataInterface() {propertyValues {Platforms {qualityLevelMask=-4}}})
     True INT=4
     False INT=0
    )
    Spawn Group=1
    Use Spawn Probability=false
    UseLoopCountLimit=false
   ) {module=SpawnBurst_Instantaneous001}
  ]
  particleSpawnScript [
   &InitializeParticle(
    Color Mode=::NewEnumerator0
    Lifetime=9999999
    Mesh Scale Mode=::NewEnumerator3
    Mesh Uniform Scale=&MakeCustomFloatFromBool(
     Bool=&IsPlatformSetActive(New Platform Set=DataInterface() {propertyValues {Platforms {qualityLevelMask=-4}}})
     True Float=&Lerp_Float(
      A=0.2
      B=0.5
      Alpha=&FloatFromCurve(
       FloatCurve=DataInterface() {propertyValues {Curve {keys [
           {arriveTangent=2.2061, interpMode=RCIM_Cubic, leaveTangent=2.2061, tangentMode=RCTM_User, time=0, value=0}
           {arriveTangent=0.0224, interpMode=RCIM_Cubic, leaveTangent=0.0224, tangentMode=RCTM_User, time=1, value=1}
          ]}}}
       CurveIndex=&RandomRangeFloat(Minimum=0, Maximum=1)
      )
     )
     False Float=&Lerp_Float(
      A=0.35
      B=0.5
      Alpha=&FloatFromCurve(
       FloatCurve=DataInterface() {propertyValues {Curve {keys [
           {arriveTangent=2.2061, interpMode=RCIM_Cubic, leaveTangent=2.2061, tangentMode=RCTM_User, time=0, value=0}
           {arriveTangent=0.0224, interpMode=RCIM_Cubic, leaveTangent=0.0224, tangentMode=RCTM_User, time=1, value=1}
          ]}}}
       CurveIndex=&RandomRangeFloat(Minimum=0, Maximum=1)
      )
     )
    )
    Position Mode=::NewEnumerator0
    Sprite Size Mode=::NewEnumerator3
    Uniform Sprite Size=30
    Write Material Random=true
   )
   &PartitionParticles(
    ID to Partition:NiagaraInt32=&MaskIntBySpawnGroup(
     Masked Int=&self:ReturnUniqueID()
     Passthrough Int=&RandomRangeInt(Minimum=0, Maximum=8)
     Spawn Group Mask=0
    )
    Number Of Partitions=9
   )
  ]
  particleUpdateScript [
   &ParticleState()
   &SkeletalMeshLocation(
    Bone / Socket Index=$Particles.PartitionParticles.Partition
    Bone Sampling Mode=::NewEnumerator4
    MaskSpawnGroup=false
    Orientation Sampling=::NewEnumerator1
    Position Sampling=::NewEnumerator2
    Sampled Orientation Coordinate Space=::Local
    Sampled Position Coordinate Space=::World
    Sampled Position Offset=&RandomRangeVector(Minimum=<-2 3 -3>, Maximum=<2 -3 3>)
    Sampled Position Offset Coordinate Space:ENiagaraSkelMeshTransforms=::NewEnumerator2
    Skeletal Mesh=$User.SkeletalMeshLeft
   )
   &SkeletalMeshLocation(
    Bone / Socket Index=1
    Bone Sampling Mode=::NewEnumerator4
    MaskSpawnGroup=false
    Orientation Sampling=::NewEnumerator1
    Position Sampling=::NewEnumerator2
    Sampled Orientation Coordinate Space=::Local
    Sampled Position Coordinate Space=::World
    Sampled Position Offset Coordinate Space:ENiagaraSkelMeshTransforms=::NewEnumerator2
    Skeletal Mesh=$User.Skeletal Mesh2
   ) {module=SkeletalMeshLocation001}
   /SetVariables(
    Particles.ArmDistance:NiagaraFloat=&Multiply_Float(A=120, B=&MakeFloatFromVector(Vector=$Particles.Scale, Channel=::NewEnumerator0))
    Particles.InterpolateSpeed:NiagaraFloat=&Lerp_Float(A=15, B=10, Alpha=$System.Timeline.NormalizedTime)
    Particles.MeshOrientationRoot:Quat4f=&OrientMeshToVector(
     Look At Direction=&TransformVector(
      Vector=&RotateVectorByQuaternion(Vector To Rotate=<1 0 0>, Quaternion=$Particles.SkeletalMeshLocation.SampledRotation)
      Destination Space=::World
     )
     Reference Axis Vector=<1 0 0>
    )
    Particles.MeshVectorRoot:Vector3f=&RotateVectorByQuaternion(Vector To Rotate=<1 0 0>, Quaternion=$Particles.SkeletalMeshLocation001.SampledRotation)
    Particles.MeshVectorWing:Vector3f=&RotateVectorByQuaternion(Vector To Rotate=<1 0 0>, Quaternion=$Particles.SkeletalMeshLocation001.SampledRotation)
    Particles.RestAxisLocalSpace:Vector3f=<0 -1 0>
    Particles.RootPosition:NiagaraPosition=$Particles.SkeletalMeshLocation.SampledPosition
    Particles.WingPos:Vector3f=$Particles.SkeletalMeshLocation001.SampledPosition
   ) {isSetParameters=true, module=SetVariables_#3}
   /SetVariables(Particles.InterpolatePos:Vector3f=&Add_Vector(
     A=&TransformVector(
      Vector=&Add_Vector(
       A=&Multiply_VectorByFloat(Vector=$Particles.MeshVectorWing, Float=-120)
       B=&Lerp_Vector(A=<20 -120 -20>, B=<-60 -140 40>, Alpha=$System.Timeline.NormalizedTime)
      )
      Destination Space=::World
     )
     B=$Particles.WingPos
    )) {isSetParameters=true, module=SetVariables_#4}
   /SetVariables(Particles.MeshOrientation=&FindQuatBetweenTwoVectors(
     From Vector=<1 0 0>
     To Vector=&TransformVector(Vector=$Particles.RestAxisLocalSpace, Destination Space=::World)
    )) {isSetParameters=true, module=SetVariables_#5}
   &InterpolateOverTime(
    Initial Vector Value:Vector3f=$Particles.RootPosition
    Rate Of Change:NiagaraFloat=$Particles.InterpolateSpeed
    Target Vector Value:Vector3f=$Particles.InterpolatePos
    Variable Type=::NewEnumerator2
   )
   &self:PositionTest(
    ArmLength:NiagaraFloat=$Particles.ArmDistance
    In:Vector3f=$Particles.InterpolateOverTime.InterpolateOverTime_V3_.Moving Average
    InterpPos:Vector3f=$Particles.InterpolatePos
    Max Angle=80
    RestAxis:Vector3f=$Particles.RestAxisLocalSpace
   ) {module=PositionTest}
   /SetVariables(Particles.OffsetAge:NiagaraFloat=&Subtract_Float(
     A=&Add_Float(A=$Particles.Age, B=&Multiply_FloatByInt(Float=100, Integer=$User.Deployed))
     B=0.5
    )) {isSetParameters=true, module=SetVariables_#6}
   /SetVariables(Particles.Position_Constraint:Vector3f=&Add_Vector(
     A=$Particles.RootPosition
     B=&Multiply_VectorByFloat(
      Vector=&NormalizeVector(
       Vector To Normalize=&Subtract_Vector(A=$Particles.PositionTest.Clamped, B=$Particles.RootPosition)
       Safe Normalize=true
       Safe Fallback Threshold=0
       Safe Fallback Vector=<0 0 0.001>
      )
      Float=$Particles.ArmDistance
     )
    )) {isSetParameters=true, module=SetVariables_#7}
   &ScaleMeshSize(Scale Factor=&MakeVector(
     X=&FloatFromCurve(
      FloatCurve=DataInterface() {propertyValues {Curve {keys [{interpMode=RCIM_Cubic, tangentMode=RCTM_User, time=0, value=0}, {interpMode=RCIM_Cubic, tangentMode=RCTM_User, time=1, value=1}]}}}
      CurveIndex=$Particles.OffsetAge
     )
     Y=&FloatFromCurve(
      FloatCurve=DataInterface() {propertyValues {Curve {keys [{tangentMode=RCTM_User, time=0.2, value=1}, {tangentMode=RCTM_User, time=1, value=1}]}}}
      CurveIndex=$Particles.OffsetAge
     )
     Z=1
    )) {module=ScaleMeshSize001}
   &FN_Timer(Increment Multiplier:NiagaraFloat=&Lerp_Float(A=0.7, B=1.2, Alpha=$System.Timeline.Time))
   &DynamicMaterialParameters(
    Index 0 Param 1=&MakeFloatFromVector(Vector=$Particles.Position_Constraint, Channel=::NewEnumerator0)
    Index 0 Param 2=&MakeFloatFromVector(Vector=$Particles.Position_Constraint, Channel=::NewEnumerator1)
    Index 0 Param 3=&MakeFloatFromVector(Vector=$Particles.Position_Constraint, Channel=::NewEnumerator2)
    Index 0 Param 4=&MakeFloatFromVector(
     Vector=&TransformVector(Vector=$Particles.RestAxisLocalSpace, Destination Space=::World)
     Channel=::NewEnumerator0
    )
    Index 1 Param 1=&MakeFloatFromVector(
     Vector=&TransformVector(Vector=$Particles.RestAxisLocalSpace, Destination Space=::World)
     Channel=::NewEnumerator1
    )
    Index 1 Param 2=&MakeFloatFromVector(
     Vector=&TransformVector(Vector=$Particles.RestAxisLocalSpace, Destination Space=::World)
     Channel=::NewEnumerator2
    )
    Index 1 Param 3=&MakeFloatFromVector(Vector=$Particles.RootPosition, Channel=::NewEnumerator0)
    Index 1 Param 4=&MakeFloatFromVector(Vector=$Particles.RootPosition, Channel=::NewEnumerator1)
    Index 2 Param 1=&MakeFloatFromVector(Vector=$Particles.RootPosition, Channel=::NewEnumerator2)
    Index 2 Param 2=1
    Index 2 Param 4=1
    Index 3 Param 1=&Add_Float(
     A=$Particles.FN_Timer.Value
     B=/Game/Effects/Niagara/DynamicInputs/Random/FN_SeededRandomRangeFloat(
      Minimum:NiagaraFloat=0
      Maximum:NiagaraFloat=50
      Random Seed:NiagaraInt32=&self:ReturnUniqueID_03()
     )
    )
    Index 3 Param 2=&MakeFloatFromVector(Vector=$Particles.Scale, Channel=::NewEnumerator0)
    Index 3 Param 3=&Lerp_Float(A=0.036, B=0, Alpha=$System.Timeline.NormalizedTime)
    Index 3 Param 4=&Lerp_Float(A=0.782, B=0, Alpha=$System.Timeline.NormalizedTime)
    Module.Param0WriteEnabled=true
    Module.Param0WriteEnabled1=true
    Module.Param0WriteEnabled2=true
    Module.Param0WriteEnabled3=true
    Module.Param1WriteEnabled=true
    Module.Param1WriteEnabled1=true
    Module.Param1WriteEnabled2=true
    Module.Param1WriteEnabled3=true
    Module.Param2WriteEnabled=true
    Module.Param2WriteEnabled1=true
    Module.Param2WriteEnabled2=true
    Module.Param2WriteEnabled3=true
    Module.Param3WriteEnabled=true
    Module.Param3WriteEnabled1=true
    Module.Param3WriteEnabled2=true
    Module.Param3WriteEnabled3=true
    Write Parameter Index 1=true
    Write Parameter Index 2=true
    Write Parameter Index 3=true
   )
   /SetVariables(Particles.Position=&Add_Vector(
     A=$Particles.RootPosition
     B=&TransformVector(
      Vector=&MakeVector(
       X=0
       Y=&FloatFromCurve(
        FloatCurve=DataInterface() {propertyValues {Curve {keys [
            {arriveTangent=-0.2987, interpMode=RCIM_Cubic, leaveTangent=-0.2987, tangentMode=RCTM_User, time=0, value=0.75}
            {arriveTangent=-0.0552, interpMode=RCIM_Cubic, leaveTangent=-0.0552, tangentMode=RCTM_User, time=1, value=0}
           ]}}}
        CurveIndex=$Particles.OffsetAge
        Scale Curve=15
       )
       Z=0
      )
      Destination Space=::World
     )
    )) {isSetParameters=true, module=SetVariables_#8}
   /SetVariables(Particles.Testend:Vector3f=$Particles.InterpolatePos) {enabled=false, isSetParameters=true, module=SetVariables_#9}
   &Color(Scale Color=<0 1 1>) {enabled=false}
   &ScaleColor(
    Scale Alpha=&MakeCustomFloatFromBool(
     Bool=$System.DefaultVarient
     True Float=1
     False Float=&FNTimeOfDayScalar(Night Scale=0.5, Morning Scale=1, Evening Scale=1)
    )
    ScaleA=true
    ScaleRGB=false
   )
  ]
  rendererClasses [&NiagaraMeshRendererProperties, &NiagaraSpriteRendererProperties]
  renderers [
   {rendererClass=&NiagaraMeshRendererProperties, rendererIndex=0}
   {rendererClass=&NiagaraSpriteRendererProperties, rendererIndex=1}
   {rendererClass=&NiagaraMeshRendererProperties, rendererIndex=2}
   {rendererClass=&NiagaraMeshRendererProperties, rendererIndex=3}
   {rendererClass=&NiagaraMeshRendererProperties, rendererIndex=4}
  ]
 }
 WingRight {
  emitterUpdateScript [
   &EmitterState(
    Inactive Response=::NewEnumerator1
    Life Cycle Mode=::NewEnumerator1
    Loop Behavior=::NewEnumerator1
    Loop Duration Mode=::NewEnumerator1
    UseLoopDelay=true
   )
   &SpawnBurst_Instantaneous(Spawn Count=9, Use Spawn Probability=false, UseLoopCountLimit=false)
   &SpawnBurst_Instantaneous(
    Spawn Count=&MakeCustomIntFromBool(
     Bool=&IsPlatformSetActive(New Platform Set=DataInterface() {propertyValues {Platforms {qualityLevelMask=-4}}})
     True INT=4
     False INT=0
    )
    Spawn Group=1
    Use Spawn Probability=false
    UseLoopCountLimit=false
   ) {module=SpawnBurst_Instantaneous001}
  ]
  particleSpawnScript [
   &InitializeParticle(
    Color Mode=::NewEnumerator0
    Lifetime=9999999
    Mesh Scale Mode=::NewEnumerator3
    Mesh Uniform Scale=&MakeCustomFloatFromBool(
     Bool=&IsPlatformSetActive(New Platform Set=DataInterface() {propertyValues {Platforms {qualityLevelMask=-4}}})
     True Float=&Lerp_Float(
      A=0.2
      B=0.5
      Alpha=&FloatFromCurve(
       FloatCurve=DataInterface() {propertyValues {Curve {keys [
           {arriveTangent=2.2061, interpMode=RCIM_Cubic, leaveTangent=2.2061, tangentMode=RCTM_User, time=0, value=0}
           {arriveTangent=0.0224, interpMode=RCIM_Cubic, leaveTangent=0.0224, tangentMode=RCTM_User, time=1, value=1}
          ]}}}
       CurveIndex=&RandomRangeFloat(Minimum=0, Maximum=1)
      )
     )
     False Float=&Lerp_Float(
      A=0.35
      B=0.5
      Alpha=&FloatFromCurve(
       FloatCurve=DataInterface() {propertyValues {Curve {keys [
           {arriveTangent=2.2061, interpMode=RCIM_Cubic, leaveTangent=2.2061, tangentMode=RCTM_User, time=0, value=0}
           {arriveTangent=0.0224, interpMode=RCIM_Cubic, leaveTangent=0.0224, tangentMode=RCTM_User, time=1, value=1}
          ]}}}
       CurveIndex=&RandomRangeFloat(Minimum=0, Maximum=1)
      )
     )
    )
    Position Mode=::NewEnumerator0
    Sprite Size Mode=::NewEnumerator3
    Uniform Sprite Size=30
    Write Material Random=true
   )
   &PartitionParticles(
    ID to Partition:NiagaraInt32=&MaskIntBySpawnGroup(
     Masked Int=&self:ReturnUniqueID()
     Passthrough Int=&RandomRangeInt(Minimum=0, Maximum=8)
     Spawn Group Mask=0
    )
    Number Of Partitions=9
   )
  ]
  particleUpdateScript [
   &ParticleState()
   &SkeletalMeshLocation(
    Bone / Socket Index=$Particles.PartitionParticles.Partition
    Bone Sampling Mode=::NewEnumerator4
    MaskSpawnGroup=false
    Orientation Sampling=::NewEnumerator1
    Position Sampling=::NewEnumerator2
    Sampled Orientation Coordinate Space=::Local
    Sampled Position Coordinate Space=::World
    Sampled Position Offset=&RandomRangeVector(Minimum=<-2 3 -3>, Maximum=<2 -3 3>)
    Sampled Position Offset Coordinate Space:ENiagaraSkelMeshTransforms=::NewEnumerator2
    Skeletal Mesh=$User.SkeletalMeshRight
   )
   &SkeletalMeshLocation(
    Bone Sampling Mode=::NewEnumerator4
    MaskSpawnGroup=false
    Orientation Sampling=::NewEnumerator1
    Position Sampling=::NewEnumerator2
    Sampled Orientation Coordinate Space=::Local
    Sampled Position Coordinate Space=::World
    Sampled Position Offset Coordinate Space:ENiagaraSkelMeshTransforms=::NewEnumerator2
    Skeletal Mesh=$User.Skeletal Mesh2
   ) {module=SkeletalMeshLocation001}
   /SetVariables(
    Particles.ArmDistance:NiagaraFloat=&Multiply_Float(A=120, B=&MakeFloatFromVector(Vector=$Particles.Scale, Channel=::NewEnumerator0))
    Particles.InterpolateSpeed:NiagaraFloat=&Lerp_Float(A=15, B=10, Alpha=$System.Timeline.NormalizedTime)
    Particles.MeshOrientationRoot:Quat4f=&OrientMeshToVector(
     Look At Direction=&TransformVector(
      Vector=&RotateVectorByQuaternion(Vector To Rotate=<1 0 0>, Quaternion=$Particles.SkeletalMeshLocation.SampledRotation)
      Destination Space=::World
     )
     Reference Axis Vector=<1 0 0>
    )
    Particles.MeshVectorRoot:Vector3f=&RotateVectorByQuaternion(Vector To Rotate=<1 0 0>, Quaternion=$Particles.SkeletalMeshLocation001.SampledRotation)
    Particles.MeshVectorWing:Vector3f=&RotateVectorByQuaternion(Vector To Rotate=<1 0 0>, Quaternion=$Particles.SkeletalMeshLocation001.SampledRotation)
    Particles.RestAxisLocalSpace:Vector3f=<0 -1 0>
    Particles.RootPosition:NiagaraPosition=$Particles.SkeletalMeshLocation.SampledPosition
    Particles.WingPos:Vector3f=$Particles.SkeletalMeshLocation001.SampledPosition
   ) {isSetParameters=true, module=SetVariables_#3}
   /SetVariables(Particles.InterpolatePos:Vector3f=&Add_Vector(
     A=&TransformVector(
      Vector=&Add_Vector(
       A=&Multiply_VectorByFloat(Vector=$Particles.MeshVectorWing, Float=120)
       B=&Lerp_Vector(A=<20 -120 20>, B=<-60 -140 -40>, Alpha=$System.Timeline.NormalizedTime)
      )
      Destination Space=::World
     )
     B=$Particles.WingPos
    )) {isSetParameters=true, module=SetVariables_#4}
   /SetVariables(Particles.MeshOrientation=&FindQuatBetweenTwoVectors(
     From Vector=<1 0 0>
     To Vector=&TransformVector(Vector=$Particles.RestAxisLocalSpace, Destination Space=::World)
    )) {isSetParameters=true, module=SetVariables_#5}
   &InterpolateOverTime(
    Initial Vector Value:Vector3f=$Particles.RootPosition
    Rate Of Change:NiagaraFloat=$Particles.InterpolateSpeed
    Target Vector Value:Vector3f=$Particles.InterpolatePos
    Variable Type=::NewEnumerator2
   )
   &self:PositionTest(
    ArmLength:NiagaraFloat=$Particles.ArmDistance
    In:Vector3f=$Particles.InterpolateOverTime.InterpolateOverTime_V3_.Moving Average
    InterpPos:Vector3f=$Particles.InterpolatePos
    Max Angle=80
    RestAxis:Vector3f=$Particles.RestAxisLocalSpace
   ) {module=PositionTest}
   /SetVariables(Particles.OffsetAge:NiagaraFloat=&Subtract_Float(
     A=&Add_Float(A=$Particles.Age, B=&Multiply_FloatByInt(Float=100, Integer=$User.Deployed))
     B=0.5
    )) {isSetParameters=true, module=SetVariables_#6}
   /SetVariables(Particles.Position_Constraint:Vector3f=&Add_Vector(
     A=$Particles.RootPosition
     B=&Multiply_VectorByFloat(
      Vector=&NormalizeVector(
       Vector To Normalize=&Subtract_Vector(A=$Particles.PositionTest.Clamped, B=$Particles.RootPosition)
       Safe Normalize=true
       Safe Fallback Threshold=0
       Safe Fallback Vector=<0 0 0.001>
      )
      Float=$Particles.ArmDistance
     )
    )) {isSetParameters=true, module=SetVariables_#7}
   &ScaleMeshSize(Scale Factor=&MakeVector(
     X=&FloatFromCurve(
      FloatCurve=DataInterface() {propertyValues {Curve {keys [{interpMode=RCIM_Cubic, tangentMode=RCTM_User, time=0, value=0}, {interpMode=RCIM_Cubic, tangentMode=RCTM_User, time=1, value=1}]}}}
      CurveIndex=$Particles.OffsetAge
     )
     Y=&FloatFromCurve(
      FloatCurve=DataInterface() {propertyValues {Curve {keys [{tangentMode=RCTM_User, time=0.2, value=1}, {tangentMode=RCTM_User, time=1, value=1}]}}}
      CurveIndex=$Particles.OffsetAge
     )
     Z=1
    )) {module=ScaleMeshSize001}
   &FN_Timer(Increment Multiplier:NiagaraFloat=&Lerp_Float(A=0.7, B=1.2, Alpha=$System.Timeline.Time))
   &DynamicMaterialParameters(
    Index 0 Param 1=&MakeFloatFromVector(Vector=$Particles.Position_Constraint, Channel=::NewEnumerator0)
    Index 0 Param 2=&MakeFloatFromVector(Vector=$Particles.Position_Constraint, Channel=::NewEnumerator1)
    Index 0 Param 3=&MakeFloatFromVector(Vector=$Particles.Position_Constraint, Channel=::NewEnumerator2)
    Index 0 Param 4=&MakeFloatFromVector(
     Vector=&TransformVector(Vector=$Particles.RestAxisLocalSpace, Destination Space=::World)
     Channel=::NewEnumerator0
    )
    Index 1 Param 1=&MakeFloatFromVector(
     Vector=&TransformVector(Vector=$Particles.RestAxisLocalSpace, Destination Space=::World)
     Channel=::NewEnumerator1
    )
    Index 1 Param 2=&MakeFloatFromVector(
     Vector=&TransformVector(Vector=$Particles.RestAxisLocalSpace, Destination Space=::World)
     Channel=::NewEnumerator2
    )
    Index 1 Param 3=&MakeFloatFromVector(Vector=$Particles.RootPosition, Channel=::NewEnumerator0)
    Index 1 Param 4=&MakeFloatFromVector(Vector=$Particles.RootPosition, Channel=::NewEnumerator1)
    Index 2 Param 1=&MakeFloatFromVector(Vector=$Particles.RootPosition, Channel=::NewEnumerator2)
    Index 2 Param 2=1
    Index 2 Param 3=-1
    Index 2 Param 4=1
    Index 3 Param 1=&Add_Float(
     A=$Particles.FN_Timer.Value
     B=&FN_SeededRandomRangeFloat(
      Minimum:NiagaraFloat=0
      Maximum:NiagaraFloat=50
      Random Seed:NiagaraInt32=&self:ReturnUniqueID_03()
     )
    )
    Index 3 Param 2=&MakeFloatFromVector(Vector=$Particles.Scale, Channel=::NewEnumerator0)
    Index 3 Param 3=&Lerp_Float(A=0.036, B=0, Alpha=$System.Timeline.NormalizedTime)
    Index 3 Param 4=&Lerp_Float(A=0.782, B=0, Alpha=$System.Timeline.NormalizedTime)
    Module.Param0WriteEnabled=true
    Module.Param0WriteEnabled1=true
    Module.Param0WriteEnabled2=true
    Module.Param0WriteEnabled3=true
    Module.Param1WriteEnabled=true
    Module.Param1WriteEnabled1=true
    Module.Param1WriteEnabled2=true
    Module.Param1WriteEnabled3=true
    Module.Param2WriteEnabled=true
    Module.Param2WriteEnabled1=true
    Module.Param2WriteEnabled2=true
    Module.Param2WriteEnabled3=true
    Module.Param3WriteEnabled=true
    Module.Param3WriteEnabled1=true
    Module.Param3WriteEnabled2=true
    Module.Param3WriteEnabled3=true
    Write Parameter Index 1=true
    Write Parameter Index 2=true
    Write Parameter Index 3=true
   )
   /SetVariables(Particles.Position=&Add_Vector(
     A=$Particles.RootPosition
     B=&TransformVector(
      Vector=&MakeVector(
       X=0
       Y=&FloatFromCurve(
        FloatCurve=DataInterface() {propertyValues {Curve {keys [
            {arriveTangent=-0.2987, interpMode=RCIM_Cubic, leaveTangent=-0.2987, tangentMode=RCTM_User, time=0, value=0.75}
            {arriveTangent=-0.0552, interpMode=RCIM_Cubic, leaveTangent=-0.0552, tangentMode=RCTM_User, time=1, value=0}
           ]}}}
        CurveIndex=$Particles.OffsetAge
        Scale Curve=15
       )
       Z=0
      )
      Destination Space=::World
     )
    )) {isSetParameters=true, module=SetVariables_#8}
   /SetVariables(Particles.Testend:Vector3f=$Particles.InterpolatePos) {enabled=false, isSetParameters=true, module=SetVariables_#9}
   &Color(Scale Color=<0 1 1>) {enabled=false}
   &ScaleColor(
    Scale Alpha=&MakeCustomFloatFromBool(
     Bool=$System.DefaultVarient
     True Float=1
     False Float=&FNTimeOfDayScalar(Night Scale=0.5, Morning Scale=1, Evening Scale=1)
    )
    ScaleA=true
    ScaleRGB=false
   )
  ]
  rendererClasses [&NiagaraMeshRendererProperties, &NiagaraSpriteRendererProperties]
  renderers [
   {rendererClass=&NiagaraMeshRendererProperties, rendererIndex=0}
   {rendererClass=&NiagaraSpriteRendererProperties, rendererIndex=1}
   {rendererClass=&NiagaraMeshRendererProperties, rendererIndex=2}
   {rendererClass=&NiagaraMeshRendererProperties, rendererIndex=3}
   {rendererClass=&NiagaraMeshRendererProperties, rendererIndex=4}
  ]
 }
 WingRight001 {
  bEnabled=false
  emitterUpdateScript [
   &EmitterState(
    Inactive Response=::NewEnumerator1
    Life Cycle Mode=::NewEnumerator1
    Loop Behavior=::NewEnumerator1
    Loop Duration Mode=::NewEnumerator1
    UseLoopDelay=true
   )
   &SpawnBurst_Instantaneous(Spawn Count=9, Use Spawn Probability=false, UseLoopCountLimit=false)
   &SpawnBurst_Instantaneous(
    Spawn Count=&MakeCustomIntFromBool(
     Bool=&IsPlatformSetActive(New Platform Set=DataInterface() {propertyValues {Platforms {qualityLevelMask=-4}}})
     True INT=4
     False INT=0
    )
    Spawn Group=1
    Use Spawn Probability=false
    UseLoopCountLimit=false
   ) {module=SpawnBurst_Instantaneous001}
  ]
  particleSpawnScript [
   &InitializeParticle(
    Color Mode=::NewEnumerator0
    Lifetime=9999999
    Mesh Scale Mode=::NewEnumerator3
    Mesh Uniform Scale=&MakeCustomFloatFromBool(
     Bool=&IsPlatformSetActive(New Platform Set=DataInterface() {propertyValues {Platforms {qualityLevelMask=-4}}})
     True Float=&Lerp_Float(
      A=0.2
      B=0.5
      Alpha=&FloatFromCurve(
       FloatCurve=DataInterface() {propertyValues {Curve {keys [
           {arriveTangent=2.2061, interpMode=RCIM_Cubic, leaveTangent=2.2061, tangentMode=RCTM_User, time=0, value=0}
           {arriveTangent=0.0224, interpMode=RCIM_Cubic, leaveTangent=0.0224, tangentMode=RCTM_User, time=1, value=1}
          ]}}}
       CurveIndex=&RandomRangeFloat(Minimum=0, Maximum=1)
      )
     )
     False Float=&Lerp_Float(
      A=0.35
      B=0.5
      Alpha=&FloatFromCurve(
       FloatCurve=DataInterface() {propertyValues {Curve {keys [
           {arriveTangent=2.2061, interpMode=RCIM_Cubic, leaveTangent=2.2061, tangentMode=RCTM_User, time=0, value=0}
           {arriveTangent=0.0224, interpMode=RCIM_Cubic, leaveTangent=0.0224, tangentMode=RCTM_User, time=1, value=1}
          ]}}}
       CurveIndex=&RandomRangeFloat(Minimum=0, Maximum=1)
      )
     )
    )
    Position Mode=::NewEnumerator0
    Sprite Size Mode=::NewEnumerator3
    Uniform Sprite Size=30
    Write Material Random=true
   )
   &PartitionParticles(
    ID to Partition:NiagaraInt32=&MaskIntBySpawnGroup(
     Masked Int=&self:ReturnUniqueID()
     Passthrough Int=&RandomRangeInt(Minimum=0, Maximum=8)
     Spawn Group Mask=0
    )
    Number Of Partitions=9
   )
  ]
  particleUpdateScript [
   &ParticleState()
   &SkeletalMeshLocation(
    Bone / Socket Index=$Particles.PartitionParticles.Partition
    Bone Sampling Mode=::NewEnumerator4
    MaskSpawnGroup=false
    Orientation Sampling=::NewEnumerator1
    Position Sampling=::NewEnumerator2
    Sampled Orientation Coordinate Space=::Local
    Sampled Position Coordinate Space=::World
    Sampled Position Offset=&RandomRangeVector(Minimum=<-2 3 -3>, Maximum=<2 -3 3>)
    Sampled Position Offset Coordinate Space:ENiagaraSkelMeshTransforms=::NewEnumerator2
    Skeletal Mesh=$User.SkeletalMeshRight
   )
   &SkeletalMeshLocation(
    Bone Sampling Mode=::NewEnumerator4
    MaskSpawnGroup=false
    Orientation Sampling=::NewEnumerator1
    Position Sampling=::NewEnumerator2
    Sampled Orientation Coordinate Space=::Local
    Sampled Position Coordinate Space=::World
    Sampled Position Offset Coordinate Space:ENiagaraSkelMeshTransforms=::NewEnumerator2
    Skeletal Mesh=$User.Skeletal Mesh2
   ) {module=SkeletalMeshLocation001}
   /SetVariables(
    Particles.ArmDistance:NiagaraFloat=&Multiply_Float(A=120, B=&MakeFloatFromVector(Vector=$Particles.Scale, Channel=::NewEnumerator0))
    Particles.InterpolateSpeed:NiagaraFloat=&Lerp_Float(A=15, B=10, Alpha=$System.Timeline.NormalizedTime)
    Particles.MeshOrientationRoot:Quat4f=&OrientMeshToVector(
     Look At Direction=&TransformVector(
      Vector=&RotateVectorByQuaternion(Vector To Rotate=<1 0 0>, Quaternion=$Particles.SkeletalMeshLocation.SampledRotation)
      Destination Space=::World
     )
     Reference Axis Vector=<1 0 0>
    )
    Particles.MeshVectorRoot:Vector3f=&RotateVectorByQuaternion(Vector To Rotate=<1 0 0>, Quaternion=$Particles.SkeletalMeshLocation001.SampledRotation)
    Particles.MeshVectorWing:Vector3f=&RotateVectorByQuaternion(Vector To Rotate=<1 0 0>, Quaternion=$Particles.SkeletalMeshLocation001.SampledRotation)
    Particles.RestAxisLocalSpace:Vector3f=<0 -1 0>
    Particles.RootPosition:NiagaraPosition=$Particles.SkeletalMeshLocation.SampledPosition
    Particles.WingPos:Vector3f=$Particles.SkeletalMeshLocation001.SampledPosition
   ) {isSetParameters=true, module=SetVariables_#3}
   /SetVariables(Particles.InterpolatePos:Vector3f=&Add_Vector(
     A=&TransformVector(
      Vector=&Add_Vector(
       A=&Multiply_VectorByFloat(Vector=$Particles.MeshVectorWing, Float=120)
       B=&Lerp_Vector(A=<20 -120 20>, B=<-60 -140 -40>, Alpha=$System.Timeline.NormalizedTime)
      )
      Destination Space=::World
     )
     B=$Particles.WingPos
    )) {isSetParameters=true, module=SetVariables_#4}
   /SetVariables(Particles.MeshOrientation=&FindQuatBetweenTwoVectors(
     From Vector=<1 0 0>
     To Vector=&TransformVector(Vector=$Particles.RestAxisLocalSpace, Destination Space=::World)
    )) {isSetParameters=true, module=SetVariables_#5}
   &InterpolateOverTime(
    Initial Vector Value:Vector3f=$Particles.RootPosition
    Rate Of Change:NiagaraFloat=$Particles.InterpolateSpeed
    Target Vector Value:Vector3f=$Particles.InterpolatePos
    Variable Type=::NewEnumerator2
   )
   &self:PositionTest(
    ArmLength:NiagaraFloat=$Particles.ArmDistance
    In:Vector3f=$Particles.InterpolateOverTime.InterpolateOverTime_V3_.Moving Average
    InterpPos:Vector3f=$Particles.InterpolatePos
    Max Angle=80
    RestAxis:Vector3f=$Particles.RestAxisLocalSpace
   ) {module=PositionTest}
   /SetVariables(Particles.OffsetAge:NiagaraFloat=&Subtract_Float(
     A=&Add_Float(A=$Particles.Age, B=&Multiply_FloatByInt(Float=100, Integer=$User.Deployed))
     B=0.5
    )) {isSetParameters=true, module=SetVariables_#6}
   /SetVariables(Particles.Position_Constraint:Vector3f=&Add_Vector(
     A=$Particles.RootPosition
     B=&Multiply_VectorByFloat(
      Vector=&NormalizeVector(
       Vector To Normalize=&Subtract_Vector(A=$Particles.PositionTest.Clamped, B=$Particles.RootPosition)
       Safe Normalize=true
       Safe Fallback Threshold=0
       Safe Fallback Vector=<0 0 0.001>
      )
      Float=$Particles.ArmDistance
     )
    )) {isSetParameters=true, module=SetVariables_#7}
   &ScaleMeshSize(Scale Factor=&MakeVector(
     X=&FloatFromCurve(
      FloatCurve=DataInterface() {propertyValues {Curve {keys [{interpMode=RCIM_Cubic, tangentMode=RCTM_User, time=0, value=0}, {interpMode=RCIM_Cubic, tangentMode=RCTM_User, time=1, value=1}]}}}
      CurveIndex=$Particles.OffsetAge
     )
     Y=&FloatFromCurve(
      FloatCurve=DataInterface() {propertyValues {Curve {keys [{tangentMode=RCTM_User, time=0.2, value=1}, {tangentMode=RCTM_User, time=1, value=1}]}}}
      CurveIndex=$Particles.OffsetAge
     )
     Z=1
    )) {module=ScaleMeshSize001}
   &FN_Timer(Increment Multiplier:NiagaraFloat=&Lerp_Float(A=0.7, B=1.2, Alpha=$System.Timeline.Time))
   &DynamicMaterialParameters(
    Index 0 Param 1=&MakeFloatFromVector(Vector=$Particles.Position_Constraint, Channel=::NewEnumerator0)
    Index 0 Param 2=&MakeFloatFromVector(Vector=$Particles.Position_Constraint, Channel=::NewEnumerator1)
    Index 0 Param 3=&MakeFloatFromVector(Vector=$Particles.Position_Constraint, Channel=::NewEnumerator2)
    Index 0 Param 4=&MakeFloatFromVector(
     Vector=&TransformVector(Vector=$Particles.RestAxisLocalSpace, Destination Space=::World)
     Channel=::NewEnumerator0
    )
    Index 1 Param 1=&MakeFloatFromVector(
     Vector=&TransformVector(Vector=$Particles.RestAxisLocalSpace, Destination Space=::World)
     Channel=::NewEnumerator1
    )
    Index 1 Param 2=&MakeFloatFromVector(
     Vector=&TransformVector(Vector=$Particles.RestAxisLocalSpace, Destination Space=::World)
     Channel=::NewEnumerator2
    )
    Index 1 Param 3=&MakeFloatFromVector(Vector=$Particles.RootPosition, Channel=::NewEnumerator0)
    Index 1 Param 4=&MakeFloatFromVector(Vector=$Particles.RootPosition, Channel=::NewEnumerator1)
    Index 2 Param 1=&MakeFloatFromVector(Vector=$Particles.RootPosition, Channel=::NewEnumerator2)
    Index 2 Param 2=1
    Index 2 Param 3=-1
    Index 2 Param 4=1
    Index 3 Param 1=&Add_Float(
     A=$Particles.FN_Timer.Value
     B=&FN_SeededRandomRangeFloat(
      Minimum:NiagaraFloat=0
      Maximum:NiagaraFloat=50
      Random Seed:NiagaraInt32=&self:ReturnUniqueID_03()
     )
    )
    Index 3 Param 2=&MakeFloatFromVector(Vector=$Particles.Scale, Channel=::NewEnumerator0)
    Index 3 Param 3=&Lerp_Float(A=0.036, B=0, Alpha=$System.Timeline.NormalizedTime)
    Index 3 Param 4=&Lerp_Float(A=0.782, B=0, Alpha=$System.Timeline.NormalizedTime)
    Module.Param0WriteEnabled=true
    Module.Param0WriteEnabled1=true
    Module.Param0WriteEnabled2=true
    Module.Param0WriteEnabled3=true
    Module.Param1WriteEnabled=true
    Module.Param1WriteEnabled1=true
    Module.Param1WriteEnabled2=true
    Module.Param1WriteEnabled3=true
    Module.Param2WriteEnabled=true
    Module.Param2WriteEnabled1=true
    Module.Param2WriteEnabled2=true
    Module.Param2WriteEnabled3=true
    Module.Param3WriteEnabled=true
    Module.Param3WriteEnabled1=true
    Module.Param3WriteEnabled2=true
    Module.Param3WriteEnabled3=true
    Write Parameter Index 1=true
    Write Parameter Index 2=true
    Write Parameter Index 3=true
   )
   /SetVariables(Particles.Position=&Add_Vector(
     A=$Particles.RootPosition
     B=&TransformVector(
      Vector=&MakeVector(
       X=0
       Y=&FloatFromCurve(
        FloatCurve=DataInterface() {propertyValues {Curve {keys [
            {arriveTangent=-0.2987, interpMode=RCIM_Cubic, leaveTangent=-0.2987, tangentMode=RCTM_User, time=0, value=0.75}
            {arriveTangent=-0.0552, interpMode=RCIM_Cubic, leaveTangent=-0.0552, tangentMode=RCTM_User, time=1, value=0}
           ]}}}
        CurveIndex=$Particles.OffsetAge
        Scale Curve=15
       )
       Z=0
      )
      Destination Space=::World
     )
    )) {isSetParameters=true, module=SetVariables_#8}
   /SetVariables(Particles.Testend:Vector3f=$Particles.InterpolatePos) {enabled=false, isSetParameters=true, module=SetVariables_#9}
   &Color(Scale Color=<0 1 1>) {enabled=false}
   &ScaleColor(
    Scale Alpha=&MakeCustomFloatFromBool(
     Bool=$System.DefaultVarient
     True Float=5
     False Float=&FNTimeOfDayScalar(Night Scale=0.5, Morning Scale=1, Evening Scale=1)
    )
    Scale RGB=<0 0 0>
    ScaleA=true
    ScaleRGB=true
   )
  ]
  rendererClasses [&NiagaraMeshRendererProperties, &NiagaraSpriteRendererProperties]
  renderers [
   {rendererClass=&NiagaraMeshRendererProperties, rendererIndex=0}
   {rendererClass=&NiagaraSpriteRendererProperties, rendererIndex=1}
   {rendererClass=&NiagaraMeshRendererProperties, rendererIndex=2}
   {rendererClass=&NiagaraMeshRendererProperties, rendererIndex=3}
   {rendererClass=&NiagaraMeshRendererProperties, rendererIndex=4}
  ]
 }
}
summary {
 systemName=NS_Glider_Nitroflow
 emitters [
  {emitterName=Caustic, rendererClasses [&NiagaraSpriteRendererProperties]}
  {emitterName=WingRight, rendererClasses [&NiagaraMeshRendererProperties, &NiagaraSpriteRendererProperties]}
  {emitterName=WingLeft, rendererClasses [&NiagaraMeshRendererProperties, &NiagaraSpriteRendererProperties]}
  {emitterName=Drops, simTarget=GPUComputeSim, rendererClasses [&NiagaraSpriteRendererProperties]}
  {
   bEnabled=false
   emitterName=WingRight001
   rendererClasses [&NiagaraMeshRendererProperties, &NiagaraSpriteRendererProperties]
  }
 ]
 userVariables [
  User.bIsFrontEnd {type=NiagaraBool, defaultValue=true}
  User.bIsFrontEndPreview {type=NiagaraBool}
  User.bisFullyDeployed {type=NiagaraBool}
  User.Deployed {type=NiagaraInt32}
  User.ForwardDot {type=NiagaraFloat, defaultValue=0, description=""}
  User.IsFrontEnd {type=NiagaraBool}
  User.isVarient {type=NiagaraFloat, defaultValue=0, description=""}
  User.RightDot {type=NiagaraFloat, defaultValue=0, description=""}
  User.Skeletal Mesh2 {
   type=NiagaraDataInterfaceSkeletalMesh
   defaultValue {
    struct=NiagaraExt_VariableValue_DataInterface
    value {dataInterface=&self:NiagaraDataInterfaceSkeletalMesh_0}
   }
  }
  User.SkeletalMesh {
   type=NiagaraDataInterfaceSkeletalMesh
   defaultValue {
    struct=NiagaraExt_VariableValue_DataInterface
    value {dataInterface=&self:NiagaraDataInterfaceSkeletalMesh_5}
   }
  }
  User.SkeletalMeshLeft {
   type=NiagaraDataInterfaceSkeletalMesh
   defaultValue {
    struct=NiagaraExt_VariableValue_DataInterface
    value {dataInterface=&self:NiagaraDataInterfaceSkeletalMesh_2}
   }
  }
  User.SkeletalMeshRight {
   type=NiagaraDataInterfaceSkeletalMesh
   defaultValue {
    struct=NiagaraExt_VariableValue_DataInterface
    value {dataInterface=&self:NiagaraDataInterfaceSkeletalMesh_1}
   }
  }
 ]
}
