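# Niagara burst effect for deploying the Nitroflow glider, layering a chromatic flare, delayed flare, expanding mesh shockwave, GPU sparks and caustic sprites.
# Caustic sprites follow bones or sockets on User.SkeletalMesh, one spawn group per particle, while sparks fly out from a sphere under drag and curl noise.
# System.DefaultVarient switches caustic, flare and spark colors between blue and warm gold.
# Spark brightness adapts to front end and time of day; shockwave intensity also reads User.Type_Transformation.
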NiagaraSystem /BRCosmetics/Gadgets/Assets/VinderTech_GliderChute/Glider_NitroFlow/FX/NS_Glider_Nitroflow_Deploy
dependencies {
 usedDataInterfaces [&NiagaraDataInterfaceCurve, &NiagaraDataInterfaceColorCurve]
 usedDynamicInputs []
 usedModules [
  &SystemState
  &self:SystemSpawnScript.SystemScriptSource.SystemScriptGraph.NiagaraNodeAssignment_0.SetVariables_#1
 ]
}
emitters {
 Caustic_Magic {
  emitterUpdateScript [
   &EmitterState(Life Cycle Mode=::NewEnumerator1, Loop Behavior=::NewEnumerator1, Loop Delay=0.5, Loop Duration=1, UseLoopDelay=false)
   &SpawnBurst_Instantaneous(Spawn Count=30, Use Spawn Probability=false)
  ]
  particleSpawnScript [
   &FN_SetSpawnGroup(Number of Spawn Groups=30)
   /Niagara/Modules/Spawn/Initialization/InitializeParticle(
    Color=&MakeCustomLinearColorFromBool(
     Bool=$System.DefaultVarient
     True Color=&LerpLinearColors(
      StartColor=<rgba 0.2222 1.0382 2 0.2>
      EndColor=<rgba 0.1667 0.7787 1.5 0.2>
      LerpFactor=&ModuloFloat(A=&MakeFloatFromInt(Int=&ReturnExecIndex()), B=2)
     )
     False Color=&LerpLinearColors(
      StartColor=<rgba 1 0.831 0.3297 0.2>
      EndColor=<rgba 0.7333 0.6094 0.2418 0.2>
      LerpFactor=&ModuloFloat(A=&MakeFloatFromInt(Int=&ReturnExecIndex()), B=2)
     )
    )
    Lifetime=5.5
    Module.Write Color=true
    Module.Write Lifetime=true
    Module.Write Mass=true
    Module.Write Material Random=true
    Module.Write Position=true
    Module.Write Scale=false
    Module.Write SpriteRotation=true
    Module.Write SpriteSize=true
    Sprite Rotation=&RandomRangeFloat(Minimum=0, Maximum=360)
    Sprite Size=<40 40>
   )
   &CameraOffset(Camera Offset Amount=15)
  ]
  particleUpdateScript [
   &ParticleState()
   &SkeletalMeshLocation(
    Bone / Socket Index=$Particles.SpawnGroup
    Bone Sampling Mode=::NewEnumerator4
    Inherit Velocity=false
    MaskSpawnGroup=false
    Skeletal Mesh=$User.SkeletalMesh
   )
   &DynamicMaterialParameters(
    Index 0 Param 1=&FloatFromCurve(FloatCurve=DataInterface() {propertyValues {Curve {keys [
         {arriveTangent=4.1785, interpMode=RCIM_Cubic, leaveTangent=4.1785, tangentMode=RCTM_User, time=0, value=0.5}
         {arriveTangent=0.2288, interpMode=RCIM_Cubic, leaveTangent=0.2288, tangentMode=RCTM_User, time=0.15, value=0.9}
         {arriveTangent=0.0224, interpMode=RCIM_Cubic, leaveTangent=0.0224, tangentMode=RCTM_User, time=1.1, value=1}
        ]}}})
    Index 0 Param 2=0.3
    Index 0 Param 3=0.3
    Index 0 Param 4=1
    Module.Param0WriteEnabled=true
    Module.Param1WriteEnabled=true
    Module.Param2WriteEnabled=true
    Module.Param3WriteEnabled=true
   )
   &SpriteRotationRate(Rotation Rate=10)
   &ScaleColor(
    Scale Alpha=&FloatFromCurve(FloatCurve=DataInterface() {propertyValues {Curve {keys [
         {arriveTangent=0.4357, interpMode=RCIM_Cubic, leaveTangent=0.4357, tangentMode=RCTM_User, time=0, value=0}
         {arriveTangent=13.0377, interpMode=RCIM_Cubic, tangentMode=RCTM_Break, time=0.25, value=1}
         {arriveTangent=-0.0405, interpMode=RCIM_Cubic, leaveTangent=-0.0405, tangentMode=RCTM_User, time=1, value=0}
        ]}}})
    Scale RGB=&MakeVectorFromLinearColorRGB(Color=&ColorFromCurve(ColorCurve=DataInterface() {propertyValues {
        AlphaCurve {keys []}
        BlueCurve {keys [<at 0.0336 1 RCIM_Cubic>,<at 0.1318 3>,<at 0.2701 1 RCIM_Cubic>]}
        GreenCurve {keys [<at 0.0336 1 RCIM_Cubic>,<at 0.1318 3>,<at 0.2701 1 RCIM_Cubic>]}
        RedCurve {keys [<at 0.0336 1 RCIM_Cubic>,<at 0.1318 3>,<at 0.2701 1 RCIM_Cubic>]}
       }}))
    ScaleA=true
    ScaleRGB=true
   )
  ]
  rendererClasses [&NiagaraSpriteRendererProperties]
  renderers [{rendererClass=&NiagaraSpriteRendererProperties}]
 }
 Chromatic_Flare {
  emitterUpdateScript [
   &EmitterState()
   &SpawnBurst_Instantaneous(Spawn Count=1, Use Spawn Probability=false)
  ]
  particleSpawnScript [
   &InitializeParticle(
    Lifetime=0.6
    Module.Write Color=true
    Module.Write Lifetime=true
    Module.Write Mass=true
    Module.Write Material Random=true
    Module.Write Position=true
    Module.Write Scale=false
    Module.Write SpriteRotation=true
    Module.Write SpriteSize=true
    Sprite Rotation=&RandomRangeFloat(Minimum=0, Maximum=360)
    Sprite Size=&Vector2DFromFloat(Value=250)
   )
   &CameraOffset(Camera Offset Amount=&FloatFromCurve(FloatCurve=DataInterface() {propertyValues {Curve {keys [<at 0 1>,<at 1 2>]}}}, Scale Curve=100))
  ]
  particleUpdateScript [
   &ParticleState()
   &ScaleSpriteSize(Scale Factor=&Vector2DFromFloat(Value=&Lerp_Float(
      A=1
      B=1.2
      Alpha=&FloatFromCurve(FloatCurve=DataInterface() {propertyValues {Curve {keys [{arriveTangent=5.3461, interpMode=RCIM_Cubic, leaveTangent=5.3461, tangentMode=RCTM_User, time=0, value=0}, <at 1 2 RCIM_Cubic>]}}})
     )))
   &DynamicMaterialParameters(Write Parameter Index 0=false) {enabled=false}
   &ScaleColor(
    Scale Alpha=&FloatFromCurve(FloatCurve=DataInterface() {propertyValues {Curve {keys [<at 0.9 1 RCIM_Cubic>,<at 1 0 RCIM_Cubic>]}}})
    Scale RGB=&VectorFromFloat(Value=&FloatFromCurve(FloatCurve=DataInterface() {propertyValues {Curve {keys [<at 0 1 RCIM_Cubic>,<at 1 0 RCIM_Cubic>]}}}))
    ScaleA=true
    ScaleRGB=true
   )
  ]
  rendererClasses [&NiagaraSpriteRendererProperties]
  renderers [
   {rendererClass=&NiagaraSpriteRendererProperties}
   {rendererClass=&NiagaraSpriteRendererProperties, rendererIndex=1}
  ]
 }
 Flare {
  emitterUpdateScript [
   &EmitterState(Life Cycle Mode=::NewEnumerator1, Loop Behavior=::NewEnumerator1, Loop Delay=0.35, Loop Duration=1, UseLoopDelay=true)
   &SpawnBurst_Instantaneous(Spawn Count=1, Use Spawn Probability=false)
  ]
  particleSpawnScript [
   /Niagara/Modules/Spawn/Initialization/V2/InitializeParticle(
    Color=&MakeCustomLinearColorFromBool(Bool=$System.DefaultVarient, True Color=<rgba 0.3443 0.6545 1 0.25>, False Color=<rgba 1 0.7622 0.4161 0.1>)
    Lifetime=3
    Position Offset=<6 -20 0>
    Sprite Rotation Mode=::NewEnumerator2
    Sprite Size Mode=::NewEnumerator3
    Uniform Sprite Size=500
    UsePositionOffset=true
    Write Material Random=true
   ) {module=InitializeParticle001}
   &CameraOffset(Camera Offset Amount=100) {enabled=false}
  ]
  particleUpdateScript [
   &ParticleState()
   &ScaleSpriteSize(Scale Factor=&Vector2DFromFloat(Value=&FloatFromCurve(FloatCurve=DataInterface() {propertyValues {Curve {keys [
          {arriveTangent=-0.3747, interpMode=RCIM_Cubic, leaveTangent=-0.3747, tangentMode=RCTM_User, time=0, value=0}
          {arriveTangent=14.5268, interpMode=RCIM_Cubic, leaveTangent=-2.4692, tangentMode=RCTM_Break, time=0.2, value=0.9847}
          {arriveTangent=0.461, interpMode=RCIM_Cubic, leaveTangent=0.461, tangentMode=RCTM_User, time=1, value=0}
         ]}}})))
   &DynamicMaterialParameters(
    Index 0 Param 1=0.5
    Index 0 Param 2=1
    Module.Param0WriteEnabled=true
    Module.Param1WriteEnabled=true
    Module.Param2WriteEnabled=true
    Module.Param3WriteEnabled=false
   )
   &ScaleColor(
    Scale Alpha=&FloatFromCurve(FloatCurve=DataInterface() {propertyValues {Curve {keys [
         {arriveTangent=-0.3747, interpMode=RCIM_Cubic, leaveTangent=-0.3747, tangentMode=RCTM_User, time=0, value=0}
         {arriveTangent=14.5268, interpMode=RCIM_Cubic, leaveTangent=-2.4692, tangentMode=RCTM_Break, time=0.2, value=0.9847}
         {arriveTangent=0.461, interpMode=RCIM_Cubic, leaveTangent=0.461, tangentMode=RCTM_User, time=1, value=0}
        ]}}})
    Scale RGB=&VectorFromFloat(Value=&FloatFromCurve(FloatCurve=DataInterface() {propertyValues {Curve {keys [{tangentMode=RCTM_User, time=0, value=1}, {tangentMode=RCTM_User, time=1, value=0}]}}}))
    ScaleA=true
    ScaleRGB=true
   )
  ]
  rendererClasses [&NiagaraSpriteRendererProperties]
  renderers [
   {rendererClass=&NiagaraSpriteRendererProperties}
   {rendererClass=&NiagaraSpriteRendererProperties, rendererIndex=1}
  ]
 }
 Spark {
  simTarget=GPUComputeSim
  emitterUpdateScript [
   &EmitterState()
   &SpawnBurst_Instantaneous(
    Spawn Count=&MakeCustomIntFromBool(Bool=$User.bIsFrontEnd, True INT=30, False INT=30)
    Use Spawn Probability=false
   )
  ]
  particleSpawnScript [
   &InitializeParticle(
    Color=<rgba 5 5 5 1>
    Lifetime Max=0.6
    Lifetime Min=0.2
    Lifetime Mode=::NewEnumerator1
    Sprite Size Max=<4 7>
    Sprite Size Min=<3 4>
    Sprite Size Mode=::NewEnumerator2
    UsePositionOffset=false
    Write Material Random=true
   )
   &SphereLocation(
    MaskSpawnGroup=false
    Override Seed=false
    Sphere Origin:NiagaraPosition=<0 0 0>
    Sphere Radius=50
    Surface Expansion Mode=::NewEnumerator1
    Surface Only=true
    UseNonUniformScale=false
    UseOffset=false
    UseSphereOrientation=false
   )
   &AddVelocityFromPoint(
    Module.Use Falloff Distance=false
    Velocity Origin:Vector3f=<0 0 0>
    Velocity Strength=&RandomRangeFloat(Minimum=600, Maximum=1500)
   )
  ]
  particleUpdateScript [
   &ParticleState()
   &ScaleColor(
    Scale Mode=::NewEnumerator1
    Scale RGBA=&MakeVector4_FromLinearColor(LinearColor=&MakeCustomLinearColorFromBool(
      Bool=$System.DefaultVarient
      True Color=&ColorFromCurve(ColorCurve=DataInterface() {propertyValues {AlphaCurve {keys [<at 0.0074 1>,<at 0.8905 1>]}, BlueCurve {keys [<at 0 2>]}, GreenCurve {keys [<at 0 0.3244>]}, RedCurve {keys [<at 0 0>]}}})
      False Color=&ColorFromCurve(ColorCurve=DataInterface() {propertyValues {
         AlphaCurve {keys [<at 0.0074 1>,<at 0.8905 1>]}
         BlueCurve {keys [<at 0 0.5253>]}
         GreenCurve {keys [<at 0 1.371>]}
         RedCurve {keys [<at 0 2>]}
        }})
     ))
    ScaleRGBA=true
   )
   &ScaleColor(
    Scale Alpha=&FNTimeOfDayScalar(Night Scale=0.5, Morning Scale=0.3, Evening Scale=0.3)
    Scale RGB=&VectorFromFloat(Value=&FN_FrontEndFloatModifier(Gameplay=10, Front End=5))
    ScaleA=true
    ScaleRGB=true
   ) {module=ScaleColor001}
   &AccelerationForce(Acceleration=&RandomRangeVector(Minimum=<0 0 0>, Maximum=<0 0 -500>))
   &Drag(Drag=&RandomRangeFloat(Minimum=6, Maximum=10), Use Linear Drag=true, Use Rotational Drag=false)
   &CurlNoiseForce(
    Noise Strength=&Multiply_Float(
     A=500
     B=&FloatFromCurve(FloatCurve=DataInterface() {propertyValues {Curve {keys [{tangentMode=RCTM_User, time=0, value=0}, <at 0.2 1>, {tangentMode=RCTM_User, time=1, value=1.3}]}}})
    )
    Pan Noise=true
   )
   &GravityForce(Gravity=&RandomRangeVector(Minimum=<0 0 0>, Maximum=<0 0 -100>))
   &SolveForcesAndVelocity(Clamp Velocity=false, Limit Acceleration=false)
   &ScaleSpriteSizeBySpeed(Max Scale Factor=<0.5 3>, Min Scale Factor=<1 0.5>)
   &ScaleSpriteSize(Scale Factor=&Vector2DFromFloat(Value=&FloatFromCurve(FloatCurve=DataInterface() {propertyValues {Curve {keys [<at 0 0.5>,<at 0.1 1>,<at 1 0.5>]}}})))
   /SetVariables(Particles.LocalPosition:Vector3f=&Add_Vector(
     A=Linked() {linkedVariable {name=Particles.Position, type=NiagaraPosition}}
     B=Linked() {linkedVariable {name=Engine.Owner.Position, type=NiagaraPosition}}
    )) {isSetParameters=true}
  ]
  rendererClasses [&NiagaraSpriteRendererProperties]
  renderers [{rendererClass=&NiagaraSpriteRendererProperties}]
 }
 Sphere_ShockWave {
  emitterUpdateScript [
   &EmitterState()
   &SpawnBurst_Instantaneous(Spawn Count=1, Use Spawn Probability=false)
  ]
  particleSpawnScript [
   /SetVariables(Particles.VisibilityTag=&MakeCustomIntFromBool(Bool=&SetBoolByFloatComparison(A=$User.Type_Transformation, B=0.65), True INT=1, False INT=0)) {enabled=false, isSetParameters=true}
   &InitializeParticle(
    Color=<rgba 0.4586 0.6546 1 1>
    Lifetime=0.25
    Mesh Scale Mode=::NewEnumerator3
    Mesh Uniform Scale=2.7
    UsePositionOffset=false
    Write Material Random=true
   ) {module=InitializeParticle001}
   &CameraOffset(Camera Offset Amount=50) {module=CameraOffset001}
  ]
  particleUpdateScript [
   &ParticleState()
   &DynamicMaterialParameters(
    Index 0 Param 1=&FloatFromCurve(FloatCurve=DataInterface() {propertyValues {Curve {keys [<at 0 1>,<at 0.3172 0.85>,<at 1 1>]}}})
    Index 0 Param 2=&FloatFromCurve(FloatCurve=DataInterface() {propertyValues {Curve {keys [<at 0 15>,<at 1 0>]}}})
    Module.Param0WriteEnabled=true
    Module.Param1WriteEnabled=true
    Module.Param2WriteEnabled=true
    Module.Param3WriteEnabled=true
   )
   &ScaleColor(
    Scale Alpha=&FloatFromCurve(FloatCurve=DataInterface() {propertyValues {Curve {keys [
         {arriveTangent=-2.1487, interpMode=RCIM_Cubic, leaveTangent=-2.1487, tangentMode=RCTM_User, time=0, value=1}
         {arriveTangent=-1.3876, interpMode=RCIM_Cubic, leaveTangent=-1.3876, tangentMode=RCTM_User, time=1.0011, value=0.002}
        ]}}})
    Scale RGB=&VectorFromFloat(Value=&FloatFromCurve(FloatCurve=DataInterface() {propertyValues {Curve {keys [
          {arriveTangent=-1.3785, interpMode=RCIM_Cubic, leaveTangent=-1.3785, tangentMode=RCTM_User, time=0.1, value=1}
          {arriveTangent=-1.4497, interpMode=RCIM_Cubic, leaveTangent=-1.4497, tangentMode=RCTM_User, time=0.4, value=0.55}
          {arriveTangent=-0.1897, interpMode=RCIM_Cubic, leaveTangent=-0.1897, tangentMode=RCTM_User, time=0.8, value=0.2}
         ]}}}))
    ScaleA=true
    ScaleRGB=true
   )
   &ScaleColor(
    Scale Alpha=&MakeCustomFloatFromBool(
     Bool=&SetBoolByFloatComparison(A=$User.Type_Transformation, B=0.65)
     True Float=1
     False Float=&FN_FrontEndFloatModifier(Gameplay=3, Front End=1.5)
    )
    Scale RGB=&VectorFromFloat(Value=&MakeCustomFloatFromBool(Bool=&SetBoolByFloatComparison(A=$User.Type_Transformation, B=0.65), True Float=1, False Float=0.1))
    ScaleA=true
    ScaleRGB=true
   ) {module=ScaleColor001}
   &ScaleMeshSize(Scale Factor=&VectorFromFloat(Value=&FloatFromCurve(FloatCurve=DataInterface() {propertyValues {Curve {keys [{arriveTangent=2.93, interpMode=RCIM_Cubic, leaveTangent=2.93, tangentMode=RCTM_User, time=0, value=1}, <at 1 2 RCIM_Cubic>]}}})))
   &CameraOffset(Camera Offset Amount=&FloatFromCurve(FloatCurve=DataInterface() {propertyValues {Curve {keys [<at 0 0>,<at 1 0.5>]}}}, Scale Curve=50))
  ]
  rendererClasses [&NiagaraMeshRendererProperties]
  renderers [
   {rendererClass=&NiagaraMeshRendererProperties, rendererIndex=0}
   {rendererClass=&NiagaraMeshRendererProperties, rendererIndex=1}
  ]
 }
}
summary {
 systemName=NS_Glider_Nitroflow_Deploy
 emitters [
  {emitterName=Chromatic_Flare, rendererClasses [&NiagaraSpriteRendererProperties]}
  {emitterName=Flare, rendererClasses [&NiagaraSpriteRendererProperties]}
  {emitterName=Sphere_ShockWave, rendererClasses [&NiagaraMeshRendererProperties]}
  {emitterName=Spark, simTarget=GPUComputeSim, rendererClasses [&NiagaraSpriteRendererProperties]}
  {emitterName=Caustic_Magic, rendererClasses [&NiagaraSpriteRendererProperties]}
 ]
 userVariables [
  User.bIsFrontEnd {type=NiagaraBool}
  User.bIsFrontEndPreview {type=NiagaraBool}
  User.IsFrontEnd {type=NiagaraBool}
  User.isVarient {type=NiagaraFloat, defaultValue=1, description=""}
  User.Return {type=NiagaraBool}
  User.SkeletalMesh {
   type=NiagaraDataInterfaceSkeletalMesh
   defaultValue {
    struct=NiagaraExt_VariableValue_DataInterface
    value {dataInterface=&self:NiagaraDataInterfaceSkeletalMesh_0}
   }
  }
  User.SkeletalMeshRight {
   type=NiagaraDataInterfaceSkeletalMesh
   defaultValue {
    struct=NiagaraExt_VariableValue_DataInterface
    value {dataInterface=&self:NiagaraDataInterfaceSkeletalMesh_1}
   }
  }
  User.Type_Transformation {type=NiagaraFloat, defaultValue=0, description=""}
 ]
}
